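# Translucent unlit Niagara particle material that draws flipbook toon smoke for the LettuceCat pickaxe effects.
# Dynamic parameter Time drives a 3x3 flipbook, whose red channel times particle alpha forms opacity.
# Dimmed particle color feeds emissive adjustments; depth fade applies only on the default and deferred shading paths.
# The IsDirectional texture blend feeds nothing, and the Picker lerp blends red with itself, so neither has any effect.
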Material /BRCosmetics/Weapons/FORT_Melee/Pickaxe_LettuceCat/FX/M_Pickaxe_LettuceCat_2DToon_Smoke
graph {
 properties {
  bUsedWithNiagaraMeshParticles=true
  bUsedWithNiagaraSprites=true
  blendMode=BLEND_Translucent
  pixelDepthOffsetMode=PDOM_Legacy
  shadingModel=MSM_Unlit
  usedShadingModels=Unlit
 }
 nodes [
  n1 = DynamicParameter() {paramNames [Time, Picker, IsDirectional, Param4]}
  LinearInterpolate(
   A=TextureSample() {texture=/Game/Weapons/FORT_Melee/Effects/ToonShader/T_2DToon_Flipbook_Smoke}.RGB
   B=TextureSample() {texture=&/T_2DToon_Flipbook_DirectionalSmoke}.RGB
   Alpha=@n1.IsDirectional
  )
  n2 = ParticleColor()
  MaterialFunctionCall(Emissive=Multiply(A=@n2.RGB, B=Constant() {r=0.3})) {materialFunction=&MF_FX_EmissiveAdjustments}
  n3 = MaterialFunctionCall(Float3=MaterialFunctionCall(
    "Animation  Phase (0-1)"=@n1.Time
    Number of Rows=Constant() {r=3}
    Number of Columns=Constant() {r=3}
    Texture=TextureObjectParameter() {parameterName=ToonSmokeTexture, texture=&T_2DToon_Flipbook_Smoke}
   ) {materialFunction=&FlipBook}.Result) {materialFunction=&BreakOutFloat3Components}
  n4 = Multiply(A=@n2.A, B=LinearInterpolate(A=@n3.R, B=@n3.R, Alpha=@n1.Picker))
  n5 = DepthFade(Opacity=@n4, FadeDistance=ScalarParameter() {defaultValue=12, parameterName=DepthFadeDis})
  ShadingPathSwitch(Default=@n5, Deferred=@n5, Forward=@n4, Mobile=@n4)
 ]
}
